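# Physics asset for the SpeedKidd medium male body that rigs the jacket for secondary motion, named as the front-end variant.
# Seven jacket chains around the back, sides and fronts each run from a small master sphere through two capsules, joined by angular-limit constraints.
# Extra constraints with free swing and twist tie each side chain's tip to its front neighbor.
# Unconstrained pelvis, spine and thigh capsules provide body shapes for the jacket.
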PhysicsAsset /BRCosmetics/Characters/Player/Male/Medium/Bodies/M_MED_SpeedKidd/Meshes/M_MED_SpeedKidd_Physics_inFrontEnd
bodies {
 dyn_jacket_bk_01 [{shapeType=Capsule, length=8, radius=6, shapeName=dyn_jacket_bk_01_capsule, center=<7 0 3.3>, rotation=<rot 0 0 90>}]
 dyn_jacket_bk_02 [{shapeType=Capsule, length=8, radius=6, shapeName=dyn_jacket_bk_02_capsule, center=<5 0 3.3>, rotation=<rot 0 0 90>}]
 dyn_jacket_bk_MSTR [{shapeType=Sphere, radius=1, shapeName=dyn_jacket_bk_MSTR_sphere}]
 dyn_jacket_l_01 [{shapeType=Capsule, length=6, radius=5.5, shapeName=dyn_jacket_l_01_capsule, center=<6 0 3.3>, rotation=<rot 0 0 90>}]
 dyn_jacket_l_02 [{shapeType=Capsule, length=17.5, radius=5.5, shapeName=dyn_jacket_l_02_capsule, center=<5 0 3.3>, rotation=<rot 0 0 90>}]
 dyn_jacket_l_MSTR [{shapeType=Sphere, radius=1, shapeName=dyn_jacket_l_MSTR_sphere}]
 dyn_jacket_l_bk_01 [{shapeType=Capsule, length=8.5, radius=5.5, shapeName=dyn_jacket_l_bk_01_capsule, center=<7 0 3.3>, rotation=<rot 0 0 90>}]
 dyn_jacket_l_bk_02 [{shapeType=Capsule, length=8, radius=6, shapeName=dyn_jacket_l_bk_02_capsule, center=<5 0 3.3>, rotation=<rot 0 0 90>}]
 dyn_jacket_l_bk_MSTR [{shapeType=Sphere, radius=1, shapeName=dyn_jacket_l_bk_MSTR_sphere}]
 dyn_jacket_l_fr_01 [{shapeType=Capsule, length=8, radius=6.5, shapeName=dyn_jacket_l_fr_01_capsule, center=<8.8 -5 3.3>, rotation=<rot 0 0 90>}]
 dyn_jacket_l_fr_02 [{shapeType=Capsule, length=8.5, radius=5.5, shapeName=dyn_jacket_l_fr_02_capsule, center=<7.6 -5 2.6>, rotation=<rot 0 0 80>}]
 dyn_jacket_l_fr_MSTR [{shapeType=Sphere, radius=1, shapeName=dyn_jacket_l_fr_MSTR_sphere}]
 dyn_jacket_r_01 [{shapeType=Capsule, length=12, radius=6, shapeName=dyn_jacket_r_01_capsule, center=<7 0 3.3>, rotation=<rot 0 0 90>}]
 dyn_jacket_r_02 [{shapeType=Capsule, length=12, radius=6, shapeName=dyn_jacket_r_02_capsule, center=<5 0 3.3>, rotation=<rot 0 0 90>}]
 dyn_jacket_r_MSTR [{shapeType=Sphere, radius=1, shapeName=dyn_jacket_r_MSTR_sphere}]
 dyn_jacket_r_bk_01 [{shapeType=Capsule, length=8.5, radius=5.5, shapeName=dyn_jacket_r_bk_01_capsule, center=<7 0 3.3>, rotation=<rot 0 0 90>}]
 dyn_jacket_r_bk_02 [{shapeType=Capsule, length=8, radius=6, shapeName=dyn_jacket_r_bk_02_capsule, center=<5 0 3.3>, rotation=<rot 0 0 90>}]
 dyn_jacket_r_bk_MSTR [{shapeType=Sphere, radius=1, shapeName=dyn_jacket_r_bk_MSTR_sphere}]
 dyn_jacket_r_fr_01 [{shapeType=Capsule, length=8, radius=6.5, shapeName=dyn_jacket_r_fr_01_capsule, center=<8.8 5 3.3>, rotation=<rot 0 0 90>}]
 dyn_jacket_r_fr_02 [{shapeType=Capsule, length=8.5, radius=5.5, shapeName=dyn_jacket_r_fr_02_capsule, center=<7.65 5 2.6>, rotation=<rot 0 0 100>}]
 dyn_jacket_r_fr_MSTR [{shapeType=Sphere, radius=1, shapeName=dyn_jacket_r_fr_MSTR_sphere}]
 pelvis [{shapeType=Capsule, length=9, radius=11, shapeName=pelvis_capsule, center=<-2.1495 -1.1492 0>}]
 spine_01 [{shapeType=Capsule, length=11, radius=9, shapeName=spine_01_capsule, center=<0 -1.2827 0>}]
 spine_03 [{shapeType=Capsule, length=7, radius=10, shapeName=spine_03_capsule, center=<0.551 0 0>}]
 thigh_l [{shapeType=Capsule, length=31, radius=8.5, shapeName=thigh_l_capsule, center=<-16.8993 1.5207 0>, rotation=<rot -90 180 180>}]
 thigh_r [{shapeType=Capsule, length=31.5, radius=8, shapeName=thigh_r_capsule, center=<16.8993 -1.732 0>, rotation=<rot 90 0 0>}]
}
constraints [
 {bone1Name=dyn_jacket_l_01, bone2Name=dyn_jacket_l_MSTR, swing1LimitDegrees=10, swing2LimitDegrees=20, twistLimitDegrees=10}
 {bone1Name=dyn_jacket_l_02, bone2Name=dyn_jacket_l_01, swing1LimitDegrees=5, swing2LimitDegrees=40, twistLimitDegrees=1}
 {bone1Name=dyn_jacket_l_bk_02, bone2Name=dyn_jacket_l_bk_01, swing1LimitDegrees=10, swing2LimitDegrees=30, twistLimitDegrees=1}
 {bone1Name=dyn_jacket_l_bk_01, bone2Name=dyn_jacket_l_bk_MSTR, swing1LimitDegrees=10, swing2LimitDegrees=15, twistLimitDegrees=10}
 {bone1Name=dyn_jacket_bk_01, bone2Name=dyn_jacket_bk_MSTR, swing1LimitDegrees=5, swing2LimitDegrees=15, twistLimitDegrees=1}
 {bone1Name=dyn_jacket_bk_02, bone2Name=dyn_jacket_bk_01, swing1LimitDegrees=5, swing2LimitDegrees=30, twistLimitDegrees=1}
 {bone1Name=dyn_jacket_r_bk_01, bone2Name=dyn_jacket_r_bk_MSTR, swing1LimitDegrees=5, swing2LimitDegrees=15, twistLimitDegrees=10}
 {bone1Name=dyn_jacket_r_bk_02, bone2Name=dyn_jacket_r_bk_01, swing1LimitDegrees=5, swing2LimitDegrees=30, twistLimitDegrees=1}
 {bone1Name=dyn_jacket_r_01, bone2Name=dyn_jacket_r_MSTR, swing1LimitDegrees=10, swing2LimitDegrees=20, twistLimitDegrees=10}
 {bone1Name=dyn_jacket_r_02, bone2Name=dyn_jacket_r_01, swing1LimitDegrees=5, swing2LimitDegrees=40, twistLimitDegrees=1}
 {bone1Name=dyn_jacket_r_fr_01, bone2Name=dyn_jacket_r_fr_MSTR, swing1LimitDegrees=10, swing2LimitDegrees=40, twistLimitDegrees=50}
 {bone1Name=dyn_jacket_r_fr_02, bone2Name=dyn_jacket_r_fr_01, swing1LimitDegrees=5, swing2LimitDegrees=50, twistLimitDegrees=10}
 {bone1Name=dyn_jacket_l_fr_01, bone2Name=dyn_jacket_l_fr_MSTR, swing1LimitDegrees=10, swing2LimitDegrees=70, twistLimitDegrees=20}
 {bone1Name=dyn_jacket_l_fr_02, bone2Name=dyn_jacket_l_fr_01, swing1LimitDegrees=5, swing2LimitDegrees=50, twistLimitDegrees=50}
 {
  bone1Name=dyn_jacket_l_02
  bone2Name=dyn_jacket_l_fr_02
  swing1LimitDegrees=45
  swing1Motion=Free
  swing2LimitDegrees=45
  swing2Motion=Free
  twistLimitDegrees=45
  twistMotion=Free
 }
 {
  bone1Name=dyn_jacket_r_02
  bone2Name=dyn_jacket_r_fr_02
  swing1LimitDegrees=45
  swing1Motion=Free
  swing2LimitDegrees=45
  swing2Motion=Free
  twistLimitDegrees=45
  twistMotion=Free
 }
]
